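# Material function forming the body of Interchange's glTF clear coat materials, outputting PBR, clear coat, anisotropy and opacity channels.
# Each texture gets transformed UVs and feeds its feature sub-function, with static switches selecting that result or a fallback.
# Iridescence optionally modifies base color, and occlusion multiplies into Specular besides having its own output.
# The coat's normal drives Normal and anisotropy, while the base normal map goes to ClearCoatNormal.
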MaterialFunction /InterchangeAssets/gltf/MaterialBodies/MF_ClearCoat_Body
graph {nodes [
  MaterialExpressionVertexTangentWS_0 = VertexTangentWS()
  n1 = Custom!(SurfaceTangent=@MaterialExpressionVertexTangentWS_0) {
   code="{\r\n##if MATERIAL_TANGENTSPACENORMAL\r\n  const float3 T = float3(1,0,0);\r\n##else\r\n  const float3 T = normalize(SurfaceTangent);\r\n##endif\r\n  return T;\r\n}"
   inputs [{inputName=SurfaceTangent, input {expression=&self:MaterialExpressionVertexTangentWS_0}}]
  }
  n2 = FunctionInput() {
   bUsePreviewValueAsDefault=true
   inputName=AnisotropyTexture_Offset_X
   inputType=FunctionInput_Scalar
   sortPriority=453
   previewValue=<0 0 0 0>
  }
  n3 = FunctionInput() {
   bUsePreviewValueAsDefault=true
   inputName=AnisotropyTexture_Offset_Y
   inputType=FunctionInput_Scalar
   sortPriority=454
   previewValue=<0 0 0 0>
  }
  n4 = FunctionInput() {
   bUsePreviewValueAsDefault=true
   inputName=AnisotropyTexture_Scale_X
   inputType=FunctionInput_Scalar
   sortPriority=455
   previewValue=<1 0 0 0>
  }
  n5 = FunctionInput() {
   bUsePreviewValueAsDefault=true
   inputName=AnisotropyTexture_Scale_Y
   inputType=FunctionInput_Scalar
   sortPriority=456
   previewValue=<1 0 0 0>
  }
  n6 = FunctionInput() {bUsePreviewValueAsDefault=true, description=Rad, inputName=AnisotropyTexture_Rotation, inputType=FunctionInput_Scalar, sortPriority=457}
  n7 = FunctionInput() {bUsePreviewValueAsDefault=true, description=Rad, inputName=AnisotropyTexture_TexCoord, inputType=FunctionInput_Scalar, sortPriority=458}
  n8 = FunctionInput() {bUsePreviewValueAsDefault=true, inputName=AnisotropyTexture_TilingMethod, inputType=FunctionInput_Vector2, sortPriority=459}
  n9 = MaterialFunctionCall!(Offset_X=@n2, Offset_Y=@n3, Scale_X=@n4, Scale_Y=@n5, Rotation=@n6, TexCoord=@n7, TilingMethod=@n8) {materialFunction=/InterchangeAssets/gltf/MaterialFunctions/MF_TransformUVs}
  n10 = MaterialFunctionCall!(
   AnisotropyTexture=FunctionInput(Preview=TextureObject() {samplerType=SAMPLERTYPE_LinearColor, texture=/InterchangeAssets/gltf/Textures/T_Anisotropy_Direction_Linear}) {bUsePreviewValueAsDefault=true, inputName=AnisotropyTexture, inputType=FunctionInput_Texture2D, sortPriority=450}
   AnisotropyStrength=FunctionInput() {bUsePreviewValueAsDefault=true, inputName=AnisotropyStrength, inputType=FunctionInput_Scalar, sortPriority=451}
   AnisotropyRotation=FunctionInput() {bUsePreviewValueAsDefault=true, inputName=AnisotropyRotation, inputType=FunctionInput_Scalar, sortPriority=452}
   Normal=NamedRerouteUsage()
   AnisotropyTexture_UVs=@n9.UVs
   AnisotropyTexture_DDX=@n9.UVs
   AnisotropyTexture_DDY=@n9.UVs
  ) {materialFunction=/InterchangeAssets/gltf/MaterialFunctions/MF_SpecularAnisotropy}
  FunctionOutput(None=StaticSwitch(
    True=@n10.Tangent
    False=@n1
    Value=StaticBoolParameter() {defaultValue=true, group=Configuration, parameterName=bHasAnisotropyTexture, sortPriority=6}
   )) {outputName=Tangent, sortPriority=9}
  n11 = NamedRerouteUsage()
  n12 = NamedRerouteUsage()
  n13 = FunctionInput() {bUsePreviewValueAsDefault=true, inputName=IridescenceIOR, inputType=FunctionInput_Scalar, sortPriority=400, previewValue=<1.3 0 0 1>}
  n14 = FunctionInput(Preview=TextureObject() {texture=/InterchangeAssets/gltf/Textures/T_White_srgb}) {bUsePreviewValueAsDefault=true, inputName=IridescenceTexture, inputType=FunctionInput_Texture2D, sortPriority=401}
  n15 = FunctionInput(Preview=TextureObject() {texture=&T_White_srgb}) {bUsePreviewValueAsDefault=true, inputName=IridescenceThicknessTexture, inputType=FunctionInput_Texture2D, sortPriority=402}
  n16 = FunctionInput() {bUsePreviewValueAsDefault=true, inputName=IridescenceFactor, inputType=FunctionInput_Scalar, sortPriority=403}
  n17 = FunctionInput() {
   bUsePreviewValueAsDefault=true
   inputName=IridescenceThicknessMinimum
   inputType=FunctionInput_Scalar
   sortPriority=404
   previewValue=<100 0 0 1>
  }
  n18 = FunctionInput() {
   bUsePreviewValueAsDefault=true
   inputName=IridescenceThicknessMaximum
   inputType=FunctionInput_Scalar
   sortPriority=405
   previewValue=<400 0 0 1>
  }
  n19 = FunctionInput() {
   bUsePreviewValueAsDefault=true
   inputName=IridescenceTexture_Offset_X
   inputType=FunctionInput_Scalar
   sortPriority=406
   previewValue=<0 0 0 0>
  }
  n20 = FunctionInput() {
   bUsePreviewValueAsDefault=true
   inputName=IridescenceTexture_Offset_Y
   inputType=FunctionInput_Scalar
   sortPriority=407
   previewValue=<0 0 0 0>
  }
  n21 = FunctionInput() {
   bUsePreviewValueAsDefault=true
   inputName=IridescenceTexture_Scale_X
   inputType=FunctionInput_Scalar
   sortPriority=408
   previewValue=<1 0 0 0>
  }
  n22 = FunctionInput() {
   bUsePreviewValueAsDefault=true
   inputName=IridescenceTexture_Scale_Y
   inputType=FunctionInput_Scalar
   sortPriority=409
   previewValue=<1 0 0 0>
  }
  n23 = FunctionInput() {bUsePreviewValueAsDefault=true, description=Rad, inputName=IridescenceTexture_Rotation, inputType=FunctionInput_Scalar, sortPriority=410}
  n24 = FunctionInput() {bUsePreviewValueAsDefault=true, inputName=IridescenceTexture_TilingMethod, inputType=FunctionInput_Vector2, sortPriority=412}
  n25 = MaterialFunctionCall(
   Offset_X=@n19
   Offset_Y=@n20
   Scale_X=@n21
   Scale_Y=@n22
   Rotation=@n23
   TexCoord=FunctionInput() {bUsePreviewValueAsDefault=true, inputName=IridescenceTexture_TexCoord, inputType=FunctionInput_Scalar, sortPriority=411}
   TilingMethod=@n24
  ) {materialFunction=&MF_TransformUVs}
  n26 = FunctionInput() {
   bUsePreviewValueAsDefault=true
   inputName=IridescenceThicknessTexture_Offset_X
   inputType=FunctionInput_Scalar
   sortPriority=413
   previewValue=<0 0 0 0>
  }
  n27 = FunctionInput() {
   bUsePreviewValueAsDefault=true
   inputName=IridescenceThicknessTexture_Offset_Y
   inputType=FunctionInput_Scalar
   sortPriority=414
   previewValue=<0 0 0 0>
  }
  n28 = FunctionInput() {
   bUsePreviewValueAsDefault=true
   inputName=IridescenceThicknessTexture_Scale_X
   inputType=FunctionInput_Scalar
   sortPriority=415
   previewValue=<1 0 0 0>
  }
  n29 = FunctionInput() {
   bUsePreviewValueAsDefault=true
   inputName=IridescenceThicknessTexture_Scale_Y
   inputType=FunctionInput_Scalar
   sortPriority=416
   previewValue=<1 0 0 0>
  }
  n30 = FunctionInput() {
   bUsePreviewValueAsDefault=true
   description=Rad
   inputName=IridescenceThicknessTexture_Rotation
   inputType=FunctionInput_Scalar
   sortPriority=417
  }
  n31 = FunctionInput() {bUsePreviewValueAsDefault=true, inputName=IridescenceThicknessTexture_TexCoord, inputType=FunctionInput_Scalar, sortPriority=418}
  n32 = FunctionInput() {bUsePreviewValueAsDefault=true, inputName=IridescenceThicknessTexture_TilingMethod, inputType=FunctionInput_Vector2, sortPriority=419}
  n33 = MaterialFunctionCall(Offset_X=@n26, Offset_Y=@n27, Scale_X=@n28, Scale_Y=@n29, Rotation=@n30, TexCoord=@n31, TilingMethod=@n32) {materialFunction=&MF_TransformUVs}
  FunctionOutput(None=StaticSwitchParameter(
    True=MaterialFunctionCall!(
     BaseColor=@n11
     IOR=@n12
     IridescenceIOR=@n13
     IridescenceTexture=@n14
     IridescenceThicknessTexture=@n15
     IridescenceFactor=@n16
     IridescenceThicknessMinimum=@n17
     IridescenceThicknessMaximum=@n18
     IridescenceTexture_UVs=@n25.UVs
     IridescenceTexture_DDX=@n25.UVs
     IridescenceTexture_DDY=@n25.UVs
     IridescenceThicknessTexture_UVs=@n33.UVs
     IridescenceThicknessTexture_DDX=@n33.UVs
     IridescenceThicknessTexture_DDY=@n33.UVs
    ) {materialFunction=/InterchangeAssets/gltf/MaterialFunctions/MF_Iridescence}.BaseColor
    False=NamedRerouteUsage()
   ) {defaultValue=true, group=Configuration, parameterName=bHasIridescence, sortPriority=9}) {outputName=BaseColor, sortPriority=1}
  n34 = FunctionInput() {bUsePreviewValueAsDefault=true, inputName=ClearCoatNormalScale, inputType=FunctionInput_Scalar, sortPriority=255, previewValue=<1 0 0 1>}
  n35 = FunctionInput() {
   bUsePreviewValueAsDefault=true
   inputName=ClearCoatTexture_Offset_X
   inputType=FunctionInput_Scalar
   sortPriority=256
   previewValue=<0 0 0 0>
  }
  n36 = FunctionInput() {
   bUsePreviewValueAsDefault=true
   inputName=ClearCoatTexture_Offset_Y
   inputType=FunctionInput_Scalar
   sortPriority=257
   previewValue=<0 0 0 0>
  }
  n37 = FunctionInput() {bUsePreviewValueAsDefault=true, inputName=ClearCoatTexture_Scale_X, inputType=FunctionInput_Scalar, sortPriority=258, previewValue=<1 0 0 0>}
  n38 = FunctionInput() {bUsePreviewValueAsDefault=true, inputName=ClearCoatTexture_Scale_Y, inputType=FunctionInput_Scalar, sortPriority=259, previewValue=<1 0 0 0>}
  n39 = FunctionInput() {bUsePreviewValueAsDefault=true, description=Rad, inputName=ClearCoatTexture_Rotation, inputType=FunctionInput_Scalar, sortPriority=260}
  n40 = FunctionInput() {bUsePreviewValueAsDefault=true, inputName=ClearCoatTexture_TilingMethod, inputType=FunctionInput_Vector2, sortPriority=262}
  n41 = MaterialFunctionCall(
   Offset_X=@n35
   Offset_Y=@n36
   Scale_X=@n37
   Scale_Y=@n38
   Rotation=@n39
   TexCoord=FunctionInput() {bUsePreviewValueAsDefault=true, inputName=ClearCoatTexture_TexCoord, inputType=FunctionInput_Scalar, sortPriority=261}
   TilingMethod=@n40
  ) {materialFunction=&MF_TransformUVs}
  n42 = FunctionInput() {
   bUsePreviewValueAsDefault=true
   inputName=ClearCoatRoughnessTexture_Offset_X
   inputType=FunctionInput_Scalar
   sortPriority=263
   previewValue=<0 0 0 0>
  }
  n43 = FunctionInput() {
   bUsePreviewValueAsDefault=true
   inputName=ClearCoatRoughnessTexture_Offset_Y
   inputType=FunctionInput_Scalar
   sortPriority=264
   previewValue=<0 0 0 0>
  }
  n44 = FunctionInput() {
   bUsePreviewValueAsDefault=true
   inputName=ClearCoatRoughnessTexture_Scale_X
   inputType=FunctionInput_Scalar
   sortPriority=265
   previewValue=<1 0 0 0>
  }
  n45 = FunctionInput() {
   bUsePreviewValueAsDefault=true
   inputName=ClearCoatRoughnessTexture_Scale_Y
   inputType=FunctionInput_Scalar
   sortPriority=266
   previewValue=<1 0 0 0>
  }
  n46 = FunctionInput() {
   bUsePreviewValueAsDefault=true
   description=Rad
   inputName=ClearCoatRoughnessTexture_Rotation
   inputType=FunctionInput_Scalar
   sortPriority=267
  }
  n47 = FunctionInput() {bUsePreviewValueAsDefault=true, inputName=ClearCoatRoughnessTexture_TexCoord, inputType=FunctionInput_Scalar, sortPriority=268}
  n48 = FunctionInput() {bUsePreviewValueAsDefault=true, inputName=ClearCoatRoughnessTexture_TilingMethod, inputType=FunctionInput_Vector2, sortPriority=269}
  n49 = MaterialFunctionCall(Offset_X=@n42, Offset_Y=@n43, Scale_X=@n44, Scale_Y=@n45, Rotation=@n46, TexCoord=@n47, TilingMethod=@n48) {materialFunction=&MF_TransformUVs}
  n50 = FunctionInput() {
   bUsePreviewValueAsDefault=true
   inputName=ClearCoatNormalTexture_Offset_X
   inputType=FunctionInput_Scalar
   sortPriority=270
   previewValue=<0 0 0 0>
  }
  n51 = FunctionInput() {
   bUsePreviewValueAsDefault=true
   inputName=ClearCoatNormalTexture_Offset_Y
   inputType=FunctionInput_Scalar
   sortPriority=271
   previewValue=<0 0 0 0>
  }
  n52 = FunctionInput() {
   bUsePreviewValueAsDefault=true
   inputName=ClearCoatNormalTexture_Scale_X
   inputType=FunctionInput_Scalar
   sortPriority=272
   previewValue=<1 0 0 0>
  }
  n53 = FunctionInput() {
   bUsePreviewValueAsDefault=true
   inputName=ClearCoatNormalTexture_Scale_Y
   inputType=FunctionInput_Scalar
   sortPriority=273
   previewValue=<1 0 0 0>
  }
  n54 = FunctionInput() {bUsePreviewValueAsDefault=true, description=Rad, inputName=ClearCoatNormalTexture_Rotation, inputType=FunctionInput_Scalar, sortPriority=274}
  n55 = FunctionInput() {bUsePreviewValueAsDefault=true, inputName=ClearCoatNormalTexture_TexCoord, inputType=FunctionInput_Scalar, sortPriority=275}
  n56 = FunctionInput() {bUsePreviewValueAsDefault=true, inputName=ClearCoatNormalTexture_TilingMethod, inputType=FunctionInput_Vector2, sortPriority=276}
  n57 = MaterialFunctionCall(Offset_X=@n50, Offset_Y=@n51, Scale_X=@n52, Scale_Y=@n53, Rotation=@n54, TexCoord=@n55, TilingMethod=@n56) {materialFunction=&MF_TransformUVs}
  n58 = MaterialFunctionCall!(
   ClearCoatTexture=FunctionInput(Preview=TextureObject() {samplerType=SAMPLERTYPE_LinearColor, texture=/InterchangeAssets/gltf/Textures/T_White_Linear}) {bUsePreviewValueAsDefault=true, inputName=ClearCoatTexture, inputType=FunctionInput_Texture2D, sortPriority=250}
   ClearCoatRoughnessTexture=FunctionInput(Preview=TextureObject() {samplerType=SAMPLERTYPE_LinearColor, texture=&T_White_Linear}) {bUsePreviewValueAsDefault=true, inputName=ClearCoatRoughnessTexture, inputType=FunctionInput_Texture2D, sortPriority=251}
   ClearCoatNormalTexture=FunctionInput(Preview=TextureObject() {samplerType=SAMPLERTYPE_Normal, texture=/InterchangeAssets/gltf/Textures/T_Generic_N}) {bUsePreviewValueAsDefault=true, inputName=ClearCoatNormalTexture, inputType=FunctionInput_Texture2D, sortPriority=252}
   ClearCoatFactor=NamedRerouteDeclaration(None=FunctionInput() {bUsePreviewValueAsDefault=true, inputName=ClearCoatFactor, inputType=FunctionInput_Scalar, sortPriority=253}) {name=Input_ClearCoatFactor, nodeColor=<rgba 0.1953 0 1 1>}
   ClearCoatRoughnessFactor=NamedRerouteDeclaration(None=FunctionInput() {bUsePreviewValueAsDefault=true, inputName=ClearCoatRoughnessFactor, inputType=FunctionInput_Scalar, sortPriority=254}) {name=Input_ClearCoatRoughnessFactor, nodeColor=<rgba 0.125 0 1 1>}
   ClearCoatNormalScale=@n34
   ClearCoatTexture_UVs=@n41.UVs
   ClearCoatTexture_DDX=@n41.UVs
   ClearCoatTexture_DDY=@n41.UVs
   ClearCoatRoughnessTexture_UVs=@n49.UVs
   ClearCoatRoughnessTexture_DDX=@n49.UVs
   ClearCoatRoughnessTexture_DDY=@n49.UVs
   ClearCoatNormalTexture_UVs=@n57.UVs
   ClearCoatNormalTexture_DDX=@n57.UVs
   ClearCoatNormalTexture_DDY=@n57.UVs
   ClearCoatNormalTexture_SinCos=@n57.UVs
  ) {materialFunction=/InterchangeAssets/gltf/MaterialFunctions/MF_Clearcoat}
  FunctionOutput(None=StaticSwitch(
    True=@n58.ClearCoatRoughness
    False=NamedRerouteUsage()
    Value=StaticBoolParameter() {defaultValue=true, group=Configuration, parameterName=bHasClearCoatTexture, sortPriority=6}
   )) {outputName=ClearCoatRoughness, sortPriority=10}
  n59 = Constant3Vector() {constant=<rgba 0 0 1 0>}
  n60 = StaticBoolParameter() {defaultValue=true, group=Configuration, parameterName=bHasClearCoatTexture, sortPriority=6}
  FunctionOutput(None=NamedRerouteDeclaration(None=StaticSwitch(True=@n58.ClearCoatTopNormal, False=@n59, Value=@n60)) {name=Normal, nodeColor=<rgba 1 0 0.5156 1>}) {outputName=Normal, sortPriority=8}
  n61 = FunctionInput(Preview=TextureObject() {samplerType=SAMPLERTYPE_Normal, texture=&T_Generic_N}) {bUsePreviewValueAsDefault=true, inputName=NormalTexture, inputType=FunctionInput_Texture2D, sortPriority=100}
  n62 = FunctionInput() {bUsePreviewValueAsDefault=true, inputName=NormalScale, inputType=FunctionInput_Scalar, sortPriority=101, previewValue=<1 0 0 1>}
  n63 = NamedRerouteDeclaration(None=@n62) {name=Input_NormalScale, nodeColor=<rgba 0 1 0.1094 1>}
  n64 = FunctionInput() {bUsePreviewValueAsDefault=true, inputName=NormalTexture_Offset_X, inputType=FunctionInput_Scalar, sortPriority=102, previewValue=<0 0 0 0>}
  n65 = FunctionInput() {bUsePreviewValueAsDefault=true, inputName=NormalTexture_Offset_Y, inputType=FunctionInput_Scalar, sortPriority=103, previewValue=<0 0 0 0>}
  n66 = FunctionInput() {bUsePreviewValueAsDefault=true, inputName=NormalTexture_Scale_X, inputType=FunctionInput_Scalar, sortPriority=104, previewValue=<1 0 0 0>}
  n67 = FunctionInput() {bUsePreviewValueAsDefault=true, inputName=NormalTexture_Scale_Y, inputType=FunctionInput_Scalar, sortPriority=105, previewValue=<1 0 0 0>}
  n68 = FunctionInput() {bUsePreviewValueAsDefault=true, description=Rad, inputName=NormalTexture_Rotation, inputType=FunctionInput_Scalar, sortPriority=106}
  n69 = MaterialFunctionCall(
   Offset_X=@n64
   Offset_Y=@n65
   Scale_X=@n66
   Scale_Y=@n67
   Rotation=@n68
   TexCoord=FunctionInput() {bUsePreviewValueAsDefault=true, inputName=NormalTexture_TexCoord, inputType=FunctionInput_Scalar, sortPriority=107}
   TilingMethod=FunctionInput() {bUsePreviewValueAsDefault=true, inputName=NormalTexture_TilingMethod, inputType=FunctionInput_Vector2, sortPriority=108}
  ) {materialFunction=&MF_TransformUVs}
  n70 = Constant3Vector() {constant=<rgba 0 0 1 0>}
  n71 = NamedRerouteUsage()
  FunctionOutput(None=StaticSwitchParameter(
    True=MaterialFunctionCall!(
     NormalTexture=@n61
     NormalScale=@n63
     NormalTexture_UVs=@n69.UVs
     NormalTexture_DDX=@n69.UVs
     NormalTexture_DDY=@n69.UVs
     NormalTexture_SinCos=@n69.UVs
    ) {materialFunction=/InterchangeAssets/gltf/MaterialFunctions/MF_Normals}.Normals
    False=MaterialFunctionCall!(Normal=@n70, Scale=@n71) {materialFunction=/InterchangeAssets/gltf/MaterialFunctions/MF_ScaleNormal}.Result
   ) {defaultValue=true, group=Configuration, parameterName=bHasNormalTexture, sortPriority=3}) {outputName=ClearCoatNormal}
  FunctionOutput(None=StaticSwitch(
    True=@n10.Anisotropy
    False=Constant()
    Value=StaticBoolParameter() {defaultValue=true, group=Configuration, parameterName=bHasAnisotropyTexture, sortPriority=6}
   )) {outputName=Anisotropy, sortPriority=5}
  n72 = FunctionInput() {bUsePreviewValueAsDefault=true, inputName=BaseColorFactor_RGB, sortPriority=1, previewValue=<1 1 1 1>}
  n73 = FunctionInput() {bUsePreviewValueAsDefault=true, inputName=BaseColorFactor_A, inputType=FunctionInput_Scalar, sortPriority=2, previewValue=<1 1 1 1>}
  n74 = FunctionInput() {bUsePreviewValueAsDefault=true, inputName=BaseColorTexture_Offset_X, inputType=FunctionInput_Scalar, sortPriority=3, previewValue=<0 0 0 0>}
  n75 = FunctionInput() {bUsePreviewValueAsDefault=true, inputName=BaseColorTexture_Offset_Y, inputType=FunctionInput_Scalar, sortPriority=4, previewValue=<0 0 0 0>}
  n76 = FunctionInput() {bUsePreviewValueAsDefault=true, inputName=BaseColorTexture_Scale_X, inputType=FunctionInput_Scalar, sortPriority=5, previewValue=<1 0 0 0>}
  n77 = FunctionInput() {bUsePreviewValueAsDefault=true, inputName=BaseColorTexture_Scale_Y, inputType=FunctionInput_Scalar, sortPriority=6, previewValue=<1 0 0 0>}
  n78 = FunctionInput() {bUsePreviewValueAsDefault=true, description=Rad, inputName=BaseColorTexture_Rotation, inputType=FunctionInput_Scalar, sortPriority=7}
  n79 = MaterialFunctionCall(
   Offset_X=@n74
   Offset_Y=@n75
   Scale_X=@n76
   Scale_Y=@n77
   Rotation=@n78
   TexCoord=FunctionInput() {bUsePreviewValueAsDefault=true, inputName=BaseColorTexture_TexCoord, inputType=FunctionInput_Scalar, sortPriority=8}
   TilingMethod=FunctionInput() {bUsePreviewValueAsDefault=true, inputName=BaseColorTexture_TilingMethod, inputType=FunctionInput_Vector2, sortPriority=9}
  ) {materialFunction=&MF_TransformUVs}
  n80 = MaterialFunctionCall!(
   BaseColorTexture=FunctionInput(Preview=TextureObject() {texture=&T_White_srgb}) {bUsePreviewValueAsDefault=true, inputName=BaseColorTexture, inputType=FunctionInput_Texture2D}
   BaseColorFactor_RGB=NamedRerouteDeclaration(None=@n72) {name=Input_BaseColorFactor_RGB, nodeColor=<rgba 0.6172 0 1 1>}
   BaseColorFactor_A=NamedRerouteDeclaration(None=@n73) {name=Input_BaseColorFactor_A, nodeColor=<rgba 0 0.6719 1 1>}
   BaseColorTexture_UVs=@n79.UVs
   BaseColorTexture_DDX=@n79.UVs
   BaseColorTexture_DDY=@n79.UVs
   AlphaCutoff=NamedRerouteDeclaration(None=FunctionInput() {bUsePreviewValueAsDefault=true, inputName=AlphaCutoff, inputType=FunctionInput_Scalar, sortPriority=10}) {name=Input_AlphaCutoff, nodeColor=<rgba 1 0 0.6094 1>}
  ) {materialFunction=/InterchangeAssets/gltf/MaterialFunctions/MF_BaseColor}
  FunctionOutput(None=StaticSwitch(
    True=@n80.Opacity
    False=NamedRerouteUsage()
    Value=StaticBoolParameter() {defaultValue=true, group=Configuration, parameterName=bHasBaseColorTexture, sortPriority=0}
   )) {outputName=Opacity, sortPriority=6}
  n81 = NamedRerouteUsage()
  n82 = NamedRerouteUsage()
  FunctionOutput(None=StaticSwitch(
    True=@n80.OpacityMask
    False=Subtract(A=@n81, B=@n82)
    Value=StaticBoolParameter() {defaultValue=true, group=Configuration, parameterName=bHasBaseColorTexture, sortPriority=0}
   )) {outputName=OpacityMask, sortPriority=7}
  n83 = FunctionInput() {bUsePreviewValueAsDefault=true, inputName=MetallicFactor, inputType=FunctionInput_Scalar, sortPriority=26, previewValue=<1 0 0 1>}
  n84 = FunctionInput() {bUsePreviewValueAsDefault=true, inputName=RoughnessFactor, inputType=FunctionInput_Scalar, sortPriority=27, previewValue=<1 0 0 1>}
  n85 = FunctionInput() {
   bUsePreviewValueAsDefault=true
   inputName=MetallicRoughnessTexture_Offset_X
   inputType=FunctionInput_Scalar
   sortPriority=28
   previewValue=<0 0 0 0>
  }
  n86 = FunctionInput() {
   bUsePreviewValueAsDefault=true
   inputName=MetallicRoughnessTexture_Offset_Y
   inputType=FunctionInput_Scalar
   sortPriority=29
   previewValue=<0 0 0 0>
  }
  n87 = FunctionInput() {
   bUsePreviewValueAsDefault=true
   inputName=MetallicRoughnessTexture_Scale_X
   inputType=FunctionInput_Scalar
   sortPriority=30
   previewValue=<1 0 0 0>
  }
  n88 = FunctionInput() {
   bUsePreviewValueAsDefault=true
   inputName=MetallicRoughnessTexture_Scale_Y
   inputType=FunctionInput_Scalar
   sortPriority=31
   previewValue=<1 0 0 0>
  }
  n89 = FunctionInput() {
   bUsePreviewValueAsDefault=true
   description=Rad
   inputName=MetallicRoughnessTexture_Rotation
   inputType=FunctionInput_Scalar
   sortPriority=32
  }
  n90 = FunctionInput() {bUsePreviewValueAsDefault=true, inputName=MetallicRoughnessTexture_TexCoord, inputType=FunctionInput_Scalar, sortPriority=33}
  n91 = FunctionInput() {bUsePreviewValueAsDefault=true, inputName=MetallicRoughnessTexture_TilingMethod, inputType=FunctionInput_Vector2, sortPriority=34}
  n92 = MaterialFunctionCall(Offset_X=@n85, Offset_Y=@n86, Scale_X=@n87, Scale_Y=@n88, Rotation=@n89, TexCoord=@n90, TilingMethod=@n91) {materialFunction=&MF_TransformUVs}
  n93 = MaterialFunctionCall!(
   MetallicRoughnessTexture=FunctionInput(Preview=TextureObject() {samplerType=SAMPLERTYPE_LinearColor, texture=&T_White_Linear}) {bUsePreviewValueAsDefault=true, inputName=MetallicRoughnessTexture, inputType=FunctionInput_Texture2D, sortPriority=25}
   MetallicFactor=NamedRerouteDeclaration(None=@n83) {name=Input_MetallicFactor, nodeColor=<rgba 0.5469 1 0 1>}
   RoughnessFactor=NamedRerouteDeclaration(None=@n84) {name=Input_RoughnessFactor, nodeColor=<rgba 0 1 0.6719 1>}
   MetallicRoughnessTexture_UVs=@n92.UVs
   MetallicRoughnessTexture_DDX=@n92.UVs
   MetallicRoughnessTexture_DDY=@n92.UVs
  ) {materialFunction=/InterchangeAssets/gltf/MaterialFunctions/MF_MetallicRoughness}
  FunctionOutput(None=StaticSwitch(
    True=@n93.Metallic
    False=NamedRerouteUsage()
    Value=StaticBoolParameter() {defaultValue=true, group=Configuration, parameterName=bHasMetallicRoughnessTexture, sortPriority=1}
   )) {outputName=Metallic, sortPriority=2}
  FunctionOutput(None=StaticSwitch(
    True=@n93.Roughness
    False=NamedRerouteUsage()
    Value=StaticBoolParameter() {defaultValue=true, group=Configuration, parameterName=bHasMetallicRoughnessTexture, sortPriority=1}
   )) {outputName=Roughness, sortPriority=4}
  n94 = TextureObject() {samplerType=SAMPLERTYPE_LinearColor, texture=&T_White_Linear}
  n95 = FunctionInput() {bUsePreviewValueAsDefault=true, inputName=OcclusionStrength, inputType=FunctionInput_Scalar, sortPriority=151, previewValue=<1 0 0 1>}
  n96 = FunctionInput() {
   bUsePreviewValueAsDefault=true
   inputName=OcclusionTexture_Offset_X
   inputType=FunctionInput_Scalar
   sortPriority=152
   previewValue=<0 0 0 0>
  }
  n97 = FunctionInput() {
   bUsePreviewValueAsDefault=true
   inputName=OcclusionTexture_Offset_Y
   inputType=FunctionInput_Scalar
   sortPriority=153
   previewValue=<0 0 0 0>
  }
  n98 = FunctionInput() {bUsePreviewValueAsDefault=true, inputName=OcclusionTexture_Scale_X, inputType=FunctionInput_Scalar, sortPriority=154, previewValue=<1 0 0 0>}
  n99 = FunctionInput() {bUsePreviewValueAsDefault=true, inputName=OcclusionTexture_Scale_Y, inputType=FunctionInput_Scalar, sortPriority=155, previewValue=<1 0 0 0>}
  n100 = FunctionInput() {bUsePreviewValueAsDefault=true, description=Rad, inputName=OcclusionTexture_Rotation, inputType=FunctionInput_Scalar, sortPriority=156}
  n101 = FunctionInput() {bUsePreviewValueAsDefault=true, description=Rad, inputName=OcclusionTexture_TexCoord, inputType=FunctionInput_Scalar, sortPriority=157}
  n102 = FunctionInput() {bUsePreviewValueAsDefault=true, inputName=OcclusionTexture_TilingMethod, inputType=FunctionInput_Vector2, sortPriority=158}
  n103 = MaterialFunctionCall(Offset_X=@n96, Offset_Y=@n97, Scale_X=@n98, Scale_Y=@n99, Rotation=@n100, TexCoord=@n101, TilingMethod=@n102) {materialFunction=&MF_TransformUVs}
  n104 = StaticSwitchParameter(
   True=MaterialFunctionCall!(
    OcclusionTexture=FunctionInput(Preview=@n94) {bUsePreviewValueAsDefault=true, inputName=OcclusionTexture, inputType=FunctionInput_Texture2D, sortPriority=150}
    OcclusionStrength=NamedRerouteDeclaration(None=@n95) {name=Input_OcclusionStrength, nodeColor=<rgba 0.0781 1 0 1>}
    OcclusionTexture_UVs=@n103.UVs
    OcclusionTexture_DDX=@n103.UVs
    OcclusionTexture_DDY=@n103.UVs
   ) {materialFunction=/InterchangeAssets/gltf/MaterialFunctions/MF_Occlusion}.Occlusion
   False=NamedRerouteUsage()
  ) {defaultValue=true, group=Configuration, parameterName=bHasOcclusionTexture, sortPriority=4}
  n105 = TextureObject() {texture=&T_White_srgb}
  n106 = FunctionInput() {bUsePreviewValueAsDefault=true, inputName=SpecularFactor, inputType=FunctionInput_Scalar, sortPriority=201, previewValue=<0.5 0 0 1>}
  n107 = FunctionInput() {bUsePreviewValueAsDefault=true, inputName=SpecularTexture_Offset_X, inputType=FunctionInput_Scalar, sortPriority=202, previewValue=<0 0 0 0>}
  n108 = FunctionInput() {bUsePreviewValueAsDefault=true, inputName=SpecularTexture_Offset_Y, inputType=FunctionInput_Scalar, sortPriority=203, previewValue=<0 0 0 0>}
  n109 = FunctionInput() {bUsePreviewValueAsDefault=true, inputName=SpecularTexture_Scale_X, inputType=FunctionInput_Scalar, sortPriority=204, previewValue=<1 0 0 0>}
  n110 = FunctionInput() {bUsePreviewValueAsDefault=true, inputName=SpecularTexture_Scale_Y, inputType=FunctionInput_Scalar, sortPriority=205, previewValue=<1 0 0 0>}
  n111 = FunctionInput() {bUsePreviewValueAsDefault=true, description=Rad, inputName=SpecularTexture_Rotation, inputType=FunctionInput_Scalar, sortPriority=206}
  n112 = FunctionInput() {bUsePreviewValueAsDefault=true, description=Rad, inputName=SpecularTexture_TexCoord, inputType=FunctionInput_Scalar, sortPriority=207}
  n113 = MaterialFunctionCall(
   Offset_X=@n107
   Offset_Y=@n108
   Scale_X=@n109
   Scale_Y=@n110
   Rotation=@n111
   TexCoord=@n112
   TilingMethod=FunctionInput() {bUsePreviewValueAsDefault=true, inputName=SpecularTexture_TilingMethod, inputType=FunctionInput_Vector2, sortPriority=208}
  ) {materialFunction=&MF_TransformUVs}
  n114 = FunctionInput() {bUsePreviewValueAsDefault=true, inputName=IOR, inputType=FunctionInput_Scalar, sortPriority=209, previewValue=<1.5 0 0 1>}
  n115 = Multiply(
   A=MaterialFunctionCall!(
    SpecularTexture=FunctionInput(Preview=@n105) {bUsePreviewValueAsDefault=true, inputName=SpecularTexture, inputType=FunctionInput_Texture2D, sortPriority=200}
    SpecularFactor=NamedRerouteDeclaration(None=@n106) {name=Input_SpecularFactor, nodeColor=<rgba 1 0 0.8203 1>}
    SpecularTexture_UVs=@n113.UVs
    SpecularTexture_DDX=@n113.UVs
    SpecularTexture_DDY=@n113.UVs
   ) {materialFunction=/InterchangeAssets/gltf/MaterialFunctions/MF_Specular}.Specular
   B=MaterialFunctionCall!(IOR=NamedRerouteDeclaration(None=@n114) {name=IOR, nodeColor=<rgba 1 0.3281 0 1>}) {materialFunction=/InterchangeAssets/gltf/MaterialFunctions/MF_IOR}.Specular
  )
  n116 = NamedRerouteUsage()
  FunctionOutput(None=Multiply(
    A=FunctionOutput(None=@n104) {outputName=Occlusion, sortPriority=11}
    B=StaticSwitchParameter(True=@n115, False=@n116) {defaultValue=true, group=Configuration, parameterName=bHasSpecularTexture, sortPriority=5}
   )) {outputName=Specular, sortPriority=3}
  n117 = FunctionInput(Preview=TextureObject() {texture=&T_White_srgb}) {bUsePreviewValueAsDefault=true, inputName=EmissiveTexture, inputType=FunctionInput_Texture2D, sortPriority=51}
  n118 = NamedRerouteDeclaration(None=FunctionInput() {bUsePreviewValueAsDefault=true, inputName=EmissiveFactor, sortPriority=52}) {name=Input_EmissiveFactor, nodeColor=<rgba 0 1 0.8828 1>}
  n119 = FunctionInput() {bUsePreviewValueAsDefault=true, inputName=EmissiveStrength, inputType=FunctionInput_Scalar, sortPriority=53, previewValue=<1 0 0 1>}
  n120 = FunctionInput() {bUsePreviewValueAsDefault=true, inputName=EmissiveTexture_Offset_X, inputType=FunctionInput_Scalar, sortPriority=54, previewValue=<0 0 0 0>}
  n121 = FunctionInput() {bUsePreviewValueAsDefault=true, inputName=EmissiveTexture_Offset_Y, inputType=FunctionInput_Scalar, sortPriority=55, previewValue=<0 0 0 0>}
  n122 = FunctionInput() {bUsePreviewValueAsDefault=true, inputName=EmissiveTexture_Scale_X, inputType=FunctionInput_Scalar, sortPriority=56, previewValue=<1 0 0 0>}
  n123 = FunctionInput() {bUsePreviewValueAsDefault=true, inputName=EmissiveTexture_Scale_Y, inputType=FunctionInput_Scalar, sortPriority=57, previewValue=<1 0 0 0>}
  n124 = FunctionInput() {bUsePreviewValueAsDefault=true, description=Rad, inputName=EmissiveTexture_Rotation, inputType=FunctionInput_Scalar, sortPriority=58}
  n125 = MaterialFunctionCall(
   Offset_X=@n120
   Offset_Y=@n121
   Scale_X=@n122
   Scale_Y=@n123
   Rotation=@n124
   TexCoord=FunctionInput() {bUsePreviewValueAsDefault=true, inputName=EmissiveTexture_TexCoord, inputType=FunctionInput_Scalar, sortPriority=59}
   TilingMethod=FunctionInput() {bUsePreviewValueAsDefault=true, inputName=EmissiveTexture_TilingMethod, inputType=FunctionInput_Vector2, sortPriority=60}
  ) {materialFunction=&MF_TransformUVs}
  FunctionOutput(None=StaticSwitchParameter(
    True=MaterialFunctionCall!(
     EmissiveTexture=@n117
     EmissiveFactor=@n118
     EmissiveStrength=@n119
     EmissiveTexture_UVs=@n125.UVs
     EmissiveTexture_DDX=@n125.UVs
     EmissiveTexture_DDY=@n125.UVs
    ) {materialFunction=/InterchangeAssets/gltf/MaterialFunctions/MF_Emissive}.Emissive
    False=NamedRerouteUsage()
   ) {defaultValue=true, group=Configuration, parameterName=bHasEmissiveTexture, sortPriority=2}) {outputName=EmissiveColor, sortPriority=5}
  FunctionOutput(None=StaticSwitch(
    True=@n58.ClearCoatAmount
    False=NamedRerouteUsage()
    Value=StaticBoolParameter() {defaultValue=true, group=Configuration, parameterName=bHasClearCoatTexture, sortPriority=6}
   )) {outputName=ClearCoat, sortPriority=9}
  NamedRerouteDeclaration(None=StaticSwitch(
    True=@n80.BaseColor
    False=NamedRerouteUsage()
    Value=StaticBoolParameter() {defaultValue=true, group=Configuration, parameterName=bHasBaseColorTexture, sortPriority=0}
   )) {name=BaseColor, nodeColor=<rgba 0.3594 0 1 1>}
 ]}
