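# Default Lit master material for glTF assets in InterchangeAssets, also enabled for skeletal meshes and mesh deformers.
# It exposes grouped parameters for base color, metallic-roughness, emissive, normals, occlusion, specular, IOR, iridescence and anisotropy.
# Each texture has offset/scale, rotation, UV channel and tiling-method controls, with neutral placeholder textures as defaults.
# All parameters feed a single MF_Default_Body material function call, which holds the shading logic.
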Material /InterchangeAssets/gltf/M_Default
graph {
 properties {
  bUsedWithMeshDeformer=true
  bUsedWithSkeletalMesh=true
  opacityMaskClipValue=0
  pixelDepthOffsetMode=PDOM_Legacy
  translucencyLightingMode=TLM_Surface
  usedShadingModels=Default Lit
 }
 nodes [
  n1 = VectorParameter() {group=BaseColor, parameterName=BaseColorFactor, sortPriority=1, defaultValue=<rgba 1 1 1 1>}
  n2 = VectorParameter() {group=BaseColor, parameterName=BaseColorTexture_OffsetScale, sortPriority=2, defaultValue=<rgba 0 0 1 1>}
  n3 = VectorParameter() {group=BaseColor, parameterName=BaseColorTexture_TilingMethod, sortPriority=5, channelNames {a="", b="", g=Tiling V, r=Tiling U}}
  n4 = TextureObjectParameter() {
   group=MetallicRoughness
   parameterName=MetallicRoughnessTexture
   samplerType=SAMPLERTYPE_LinearColor
   sortPriority=0
   texture=/InterchangeAssets/gltf/Textures/T_White_Linear
  }
  n5 = VectorParameter() {group=MetallicRoughness, parameterName=MetallicRoughnessTexture_OffsetScale, sortPriority=3, defaultValue=<rgba 0 0 1 1>}
  n6 = VectorParameter() {
   group=MetallicRoughness
   parameterName=MetallicRoughnessTexture_TilingMethod
   sortPriority=6
   channelNames {a="", b="", g=Tiling V, r=Tiling U}
  }
  n7 = VectorParameter() {group=Emissive, parameterName=EmissiveTexture_OffsetScale, sortPriority=3, defaultValue=<rgba 0 0 1 1>}
  n8 = VectorParameter() {group=Emissive, parameterName=EmissiveTexture_TilingMethod, sortPriority=6, channelNames {a="", b="", g=Tiling V, r=Tiling U}}
  n9 = TextureObjectParameter() {
   group=Normals
   parameterName=NormalTexture
   samplerType=SAMPLERTYPE_Normal
   sortPriority=0
   texture=/InterchangeAssets/gltf/Textures/T_Generic_N
  }
  n10 = VectorParameter() {group=Normals, parameterName=NormalTexture_OffsetScale, sortPriority=2, defaultValue=<rgba 0 0 1 1>}
  n11 = VectorParameter() {group=Normals, parameterName=NormalTexture_TilingMethod, sortPriority=5, channelNames {a="", b="", g=Tiling V, r=Tiling U}}
  n12 = TextureObjectParameter() {
   group=Occlusion
   parameterName=OcclusionTexture
   samplerType=SAMPLERTYPE_LinearColor
   sortPriority=0
   texture=&T_White_Linear
  }
  n13 = VectorParameter() {group=Occlusion, parameterName=OcclusionTexture_OffsetScale, sortPriority=2, defaultValue=<rgba 0 0 1 1>}
  n14 = VectorParameter() {group=Occlusion, parameterName=OcclusionTexture_TilingMethod, sortPriority=5, channelNames {a="", b="", g=Tiling V, r=Tiling U}}
  n15 = VectorParameter() {group=Specular, parameterName=SpecularTexture_OffsetScale, sortPriority=2, defaultValue=<rgba 0 0 1 1>}
  n16 = VectorParameter() {group=Specular, parameterName=SpecularTexture_TilingMethod, sortPriority=5, channelNames {a="", b="", g=Tiling V, r=Tiling U}}
  n17 = TextureObjectParameter() {group=Iridescence, parameterName=IridescenceTexture, sortPriority=1, texture=/InterchangeAssets/gltf/Textures/T_White_srgb}
  n18 = TextureObjectParameter() {group=Iridescence, parameterName=IridescenceThicknessTexture, sortPriority=2, texture=&T_White_srgb}
  n19 = VectorParameter() {group=Iridescence, parameterName=IridescenceTexture_OffsetScale, sortPriority=6, defaultValue=<rgba 0 0 1 1>}
  n20 = VectorParameter() {group=Iridescence, parameterName=IridescenceThicknessTexture_OffsetScale, sortPriority=10, defaultValue=<rgba 0 0 1 1>}
  n21 = TextureObjectParameter() {
   group=Anisotropy
   parameterName=AnisotropyTexture
   samplerType=SAMPLERTYPE_LinearColor
   sortPriority=0
   texture=/InterchangeAssets/gltf/Textures/T_Anisotropy_Direction_Linear
  }
  n22 = VectorParameter() {group=Anisotropy, parameterName=AnisotropyTexture_OffsetScale, sortPriority=3, defaultValue=<rgba 0 0 1 1>}
  MaterialFunctionCall!(
   BaseColorTexture=TextureObjectParameter() {group=BaseColor, parameterName=BaseColorTexture, sortPriority=0, texture=&T_White_srgb}
   BaseColorFactor_RGB=@n1.RGB
   BaseColorFactor_A=@n1.RGB
   BaseColorTexture_Offset_X=@n2.R
   BaseColorTexture_Offset_Y=@n2.R
   BaseColorTexture_Scale_X=@n2.R
   BaseColorTexture_Scale_Y=@n2.R
   BaseColorTexture_Rotation=ScalarParameter() {group=BaseColor, parameterName=BaseColorTexture_Rotation, sliderMax=6.2832, sortPriority=3}
   BaseColorTexture_TexCoord=ScalarParameter() {group=BaseColor, parameterName=BaseColorTexture_TexCoord, sliderMax=3, sortPriority=4}
   BaseColorTexture_TilingMethod=@n3
   AlphaCutoff=ScalarParameter() {group=BaseColor, parameterName=AlphaCutoff, sliderMax=1, sortPriority=6}
   MetallicRoughnessTexture=@n4
   MetallicFactor=ScalarParameter() {defaultValue=1, group=MetallicRoughness, parameterName=MetallicFactor, sliderMax=1, sortPriority=1}
   RoughnessFactor=ScalarParameter() {defaultValue=1, group=MetallicRoughness, parameterName=RoughnessFactor, sliderMax=1, sortPriority=2}
   MetallicRoughnessTexture_Offset_X=@n5.R
   MetallicRoughnessTexture_Offset_Y=@n5.R
   MetallicRoughnessTexture_Scale_X=@n5.R
   MetallicRoughnessTexture_Scale_Y=@n5.R
   MetallicRoughnessTexture_Rotation=ScalarParameter() {group=MetallicRoughness, parameterName=MetallicRoughnessTexture_Rotation, sliderMax=6.2832, sortPriority=4}
   MetallicRoughnessTexture_TexCoord=ScalarParameter() {group=MetallicRoughness, parameterName=MetallicRoughnessTexture_TexCoord, sliderMax=3, sortPriority=5}
   MetallicRoughnessTexture_TilingMethod=@n6
   EmissiveTexture=TextureObjectParameter() {group=Emissive, parameterName=EmissiveTexture, sortPriority=0, texture=&T_White_srgb}
   EmissiveFactor=VectorParameter() {group=Emissive, parameterName=EmissiveFactor, sortPriority=1, defaultValue=<rgba 0 0 0 1>}.RGB
   EmissiveStrength=ScalarParameter() {defaultValue=1, group=Emissive, parameterName=EmissiveStrength, sortPriority=2}
   EmissiveTexture_Offset_X=@n7.R
   EmissiveTexture_Offset_Y=@n7.R
   EmissiveTexture_Scale_X=@n7.R
   EmissiveTexture_Scale_Y=@n7.R
   EmissiveTexture_Rotation=ScalarParameter() {group=Emissive, parameterName=EmissiveTexture_Rotation, sliderMax=6.2832, sortPriority=4}
   EmissiveTexture_TexCoord=ScalarParameter() {group=Emissive, parameterName=EmissiveTexture_TexCoord, sliderMax=3, sortPriority=5}
   EmissiveTexture_TilingMethod=@n8
   NormalTexture=@n9
   NormalScale=ScalarParameter() {defaultValue=1, group=Normals, parameterName=NormalScale, sortPriority=1}
   NormalTexture_Offset_X=@n10.R
   NormalTexture_Offset_Y=@n10.R
   NormalTexture_Scale_X=@n10.R
   NormalTexture_Scale_Y=@n10.R
   NormalTexture_Rotation=ScalarParameter() {group=Normals, parameterName=NormalTexture_Rotation, sliderMax=6.2832, sortPriority=3}
   NormalTexture_TexCoord=ScalarParameter() {group=Normals, parameterName=NormalTexture_TexCoord, sliderMax=3, sortPriority=4}
   NormalTexture_TilingMethod=@n11
   OcclusionTexture=@n12
   OcclusionStrength=ScalarParameter() {defaultValue=1, group=Occlusion, parameterName=OcclusionStrength, sliderMax=1, sortPriority=1}
   OcclusionTexture_Offset_X=@n13.R
   OcclusionTexture_Offset_Y=@n13.R
   OcclusionTexture_Scale_X=@n13.R
   OcclusionTexture_Scale_Y=@n13.R
   OcclusionTexture_Rotation=ScalarParameter() {group=Occlusion, parameterName=OcclusionTexture_Rotation, sliderMax=6.2832, sortPriority=3}
   OcclusionTexture_TexCoord=ScalarParameter() {group=Occlusion, parameterName=OcclusionTexture_TexCoord, sliderMax=3, sortPriority=4}
   OcclusionTexture_TilingMethod=@n14
   SpecularTexture=TextureObjectParameter() {group=Specular, parameterName=SpecularTexture, sortPriority=0, texture=&T_White_srgb}
   SpecularFactor=ScalarParameter() {defaultValue=1, group=Specular, parameterName=SpecularFactor, sliderMax=1, sortPriority=1}
   SpecularTexture_Offset_X=@n15.R
   SpecularTexture_Offset_Y=@n15.R
   SpecularTexture_Scale_X=@n15.R
   SpecularTexture_Scale_Y=@n15.R
   SpecularTexture_Rotation=ScalarParameter() {group=Specular, parameterName=SpecularTexture_Rotation, sliderMax=6.2832, sortPriority=3}
   SpecularTexture_TexCoord=ScalarParameter() {group=Specular, parameterName=SpecularTexture_TexCoord, sliderMax=3, sortPriority=4}
   SpecularTexture_TilingMethod=@n16
   IOR=ScalarParameter() {defaultValue=1.5, group=IOR, parameterName=IOR, sliderMax=2, sliderMin=1, sortPriority=0}
   IridescenceIOR=ScalarParameter() {defaultValue=1.3, group=Iridescence, parameterName=IridescenceIOR, sortPriority=0}
   IridescenceTexture=@n17
   IridescenceThicknessTexture=@n18
   IridescenceFactor=ScalarParameter() {group=Iridescence, parameterName=IridescenceFactor, sortPriority=3}
   IridescenceThicknessMinimum=ScalarParameter() {defaultValue=100, group=Iridescence, parameterName=IridescenceThicknessMinimum, sortPriority=4}
   IridescenceThicknessMaximum=ScalarParameter() {defaultValue=400, group=Iridescence, parameterName=IridescenceThicknessMaximum, sortPriority=5}
   IridescenceTexture_Offset_X=@n19.R
   IridescenceTexture_Offset_Y=@n19.R
   IridescenceTexture_Scale_X=@n19.R
   IridescenceTexture_Scale_Y=@n19.R
   IridescenceTexture_Rotation=ScalarParameter() {group=Iridescence, parameterName=IridescenceTexture_Rotation, sortPriority=7}
   IridescenceTexture_TexCoord=ScalarParameter() {group=Iridescence, parameterName=IridescenceTexture_TexCoord, sortPriority=8}
   IridescenceTexture_TilingMethod=VectorParameter() {group=Iridescence, parameterName=IridescenceTexture_TilingMethod, sortPriority=9}.RGB
   IridescenceThicknessTexture_Offset_X=@n20.R
   IridescenceThicknessTexture_Offset_Y=@n20.R
   IridescenceThicknessTexture_Scale_X=@n20.R
   IridescenceThicknessTexture_Scale_Y=@n20.R
   IridescenceThicknessTexture_Rotation=ScalarParameter() {group=Iridescence, parameterName=IridescenceThicknessTexture_Rotation, sortPriority=11}
   IridescenceThicknessTexture_TexCoord=ScalarParameter() {group=Iridescence, parameterName=IridescenceThicknessTexture_TexCoord, sortPriority=12}
   IridescenceThicknessTexture_TilingMethod=VectorParameter() {group=Iridescence, parameterName=IridescenceThicknessTexture_TilingMethod, sortPriority=13}.RGB
   AnisotropyTexture=@n21
   AnisotropyStrength=ScalarParameter() {group=Anisotropy, parameterName=AnisotropyStrength, sortPriority=1}
   AnisotropyRotation=ScalarParameter() {group=Anisotropy, parameterName=AnisotropyRotation, sortPriority=2}
   AnisotropyTexture_Offset_X=@n22.R
   AnisotropyTexture_Offset_Y=@n22.R
   AnisotropyTexture_Scale_X=@n22.R
   AnisotropyTexture_Scale_Y=@n22.R
   AnisotropyTexture_Rotation=ScalarParameter() {group=Anisotropy, parameterName=AnisotropyTexture_Rotation, sortPriority=4}
   AnisotropyTexture_TexCoord=ScalarParameter() {group=Anisotropy, parameterName=AnisotropyTexture_TexCoord, sortPriority=5}
   AnisotropyTexture_TilingMethod=VectorParameter() {group=Anisotropy, parameterName=AnisotropyTexture_TilingMethod, sortPriority=6}.RGB
  ) {materialFunction=/InterchangeAssets/gltf/MaterialBodies/MF_Default_Body}
 ]
}
