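# Material function that forms the body of glTF specular-glossiness materials in InterchangeAssets, converting that workflow to Unreal's metallic-roughness outputs.
# MF_TransformUVs positions each texture; MF_DiffuseSpecGloss feeds MF_SpecGlossToMetalRoughness for base color, metallic and roughness.
# Helper functions supply normal, occlusion and emissive, while OpacityMask is opacity minus AlphaCutoff.
# Static switches per texture set select between sampled results and factor-only or constant fallbacks, several via named reroutes.
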MaterialFunction /InterchangeAssets/gltf/MaterialBodies/MF_SpecularGlossiness_Body
graph {nodes [
  n1 = FunctionInput() {bUsePreviewValueAsDefault=true, inputName=DiffuseFactor_RGB, sortPriority=2, previewValue=<1 1 1 0>}
  n2 = FunctionInput() {bUsePreviewValueAsDefault=true, inputName=DiffuseFactor_A, inputType=FunctionInput_Scalar, sortPriority=3, previewValue=<1 0 0 0>}
  n3 = FunctionInput() {bUsePreviewValueAsDefault=true, inputName=GlossinessFactor, inputType=FunctionInput_Scalar, sortPriority=5, previewValue=<1 0 0 0>}
  n4 = FunctionInput() {bUsePreviewValueAsDefault=true, inputName=DiffuseTexture_Offset_X, inputType=FunctionInput_Scalar, sortPriority=6, previewValue=<0 0 0 0>}
  n5 = FunctionInput() {bUsePreviewValueAsDefault=true, inputName=DiffuseTexture_Offset_Y, inputType=FunctionInput_Scalar, sortPriority=7, previewValue=<0 0 0 0>}
  n6 = FunctionInput() {bUsePreviewValueAsDefault=true, inputName=DiffuseTexture_Scale_X, inputType=FunctionInput_Scalar, sortPriority=8, previewValue=<1 0 0 0>}
  n7 = FunctionInput() {bUsePreviewValueAsDefault=true, inputName=DiffuseTexture_Scale_Y, inputType=FunctionInput_Scalar, sortPriority=9, previewValue=<1 0 0 0>}
  n8 = FunctionInput() {bUsePreviewValueAsDefault=true, description=Rad, inputName=DiffuseTexture_Rotation, inputType=FunctionInput_Scalar, sortPriority=10}
  n9 = MaterialFunctionCall!(
   Offset_X=@n4
   Offset_Y=@n5
   Scale_X=@n6
   Scale_Y=@n7
   Rotation=@n8
   TexCoord=FunctionInput() {bUsePreviewValueAsDefault=true, inputName=DiffuseTexture_TexCoord, inputType=FunctionInput_Scalar, sortPriority=11}
   TilingMethod=FunctionInput() {bUsePreviewValueAsDefault=true, inputName=DiffuseTexture_TilingMethod, inputType=FunctionInput_Vector2, sortPriority=12}
  ) {materialFunction=/InterchangeAssets/gltf/MaterialFunctions/MF_TransformUVs}
  n10 = FunctionInput() {
   bUsePreviewValueAsDefault=true
   inputName=SpecularGlossinessTexture_Offset_X
   inputType=FunctionInput_Scalar
   sortPriority=13
   previewValue=<0 0 0 0>
  }
  n11 = FunctionInput() {
   bUsePreviewValueAsDefault=true
   inputName=SpecularGlossinessTexture_Offset_Y
   inputType=FunctionInput_Scalar
   sortPriority=14
   previewValue=<0 0 0 0>
  }
  n12 = FunctionInput() {
   bUsePreviewValueAsDefault=true
   inputName=SpecularGlossinessTexture_Scale_X
   inputType=FunctionInput_Scalar
   sortPriority=15
   previewValue=<1 0 0 0>
  }
  n13 = FunctionInput() {
   bUsePreviewValueAsDefault=true
   inputName=SpecularGlossinessTexture_Scale_Y
   inputType=FunctionInput_Scalar
   sortPriority=16
   previewValue=<1 0 0 0>
  }
  n14 = FunctionInput() {
   bUsePreviewValueAsDefault=true
   description=Rad
   inputName=SpecularGlossinessTexture_Rotation
   inputType=FunctionInput_Scalar
   sortPriority=17
  }
  n15 = FunctionInput() {bUsePreviewValueAsDefault=true, inputName=SpecularGlossinessTexture_TexCoord, inputType=FunctionInput_Scalar, sortPriority=18}
  n16 = FunctionInput() {bUsePreviewValueAsDefault=true, inputName=SpecularGlossinessTexture_TilingMethod, inputType=FunctionInput_Vector2, sortPriority=19}
  n17 = MaterialFunctionCall(Offset_X=@n10, Offset_Y=@n11, Scale_X=@n12, Scale_Y=@n13, Rotation=@n14, TexCoord=@n15, TilingMethod=@n16) {materialFunction=&MF_TransformUVs}
  n18 = MaterialFunctionCall!(
   DiffuseTexture=FunctionInput(Preview=TextureObject() {texture=/InterchangeAssets/gltf/Textures/T_White_srgb}) {bUsePreviewValueAsDefault=true, inputName=DiffuseTexture, inputType=FunctionInput_Texture2D}
   SpecularGlossinessTexture=FunctionInput(Preview=TextureObject() {texture=&T_White_srgb}) {bUsePreviewValueAsDefault=true, inputName=SpecularGlossinessTexture, inputType=FunctionInput_Texture2D, sortPriority=1}
   DiffuseFactor_RGB=NamedRerouteDeclaration(None=@n1) {name=Input_DiffuseFactor_RGB, nodeColor=<rgba 1 0 0.6094 1>}
   DiffuseFactor_A=NamedRerouteDeclaration(None=@n2) {name=Input_DiffuseFactor_A, nodeColor=<rgba 1 0 0.4453 1>}
   SpecFactor=NamedRerouteDeclaration(None=FunctionInput() {bUsePreviewValueAsDefault=true, inputName=SpecFactor, sortPriority=4, previewValue=<1 1 1 0>}) {name=Input_SpecFactor, nodeColor=<rgba 0 0.7188 1 1>}
   GlossinessFactor=NamedRerouteDeclaration(None=@n3) {name=Input_GlossinessFactor, nodeColor=<rgba 0.9688 1 0 1>}
   DiffuseTexture_UVs=@n9.UVs
   DiffuseTexture_DDX=@n9.UVs
   DiffuseTexture_DDY=@n9.UVs
   SpecularGlossinessTexture_UVs=@n17.UVs
   SpecularGlossinessTexture_DDX=@n17.UVs
   SpecularGlossinessTexture_DDY=@n17.UVs
  ) {materialFunction=/InterchangeAssets/gltf/MaterialFunctions/MF_DiffuseSpecGloss}
  n19 = MaterialFunctionCall!(DiffuseColor=@n18.DiffuseColor, SpecularColor=@n18.DiffuseColor, Glossiness=@n18.DiffuseColor) {materialFunction=/InterchangeAssets/gltf/MaterialFunctions/MF_SpecGlossToMetalRoughness}
  FunctionOutput(None=StaticSwitch(
    True=@n19."BaseColor (V3)"
    False=NamedRerouteUsage()
    Value=StaticBoolParameter() {defaultValue=true, group=Configuration, parameterName=bHasDiffuseSpecGlossTexture, sortPriority=0}
   )) {outputName=BaseColor, sortPriority=1}
  n20 = FunctionInput(Preview=TextureObject() {samplerType=SAMPLERTYPE_Normal, texture=/InterchangeAssets/gltf/Textures/T_Generic_N}) {bUsePreviewValueAsDefault=true, inputName=NormalTexture, inputType=FunctionInput_Texture2D, sortPriority=100}
  n21 = FunctionInput() {bUsePreviewValueAsDefault=true, inputName=NormalScale, inputType=FunctionInput_Scalar, sortPriority=101, previewValue=<1 0 0 1>}
  n22 = NamedRerouteDeclaration(None=@n21) {name=Input_NormalScale, nodeColor=<rgba 1 0 0.1406 1>}
  n23 = FunctionInput() {bUsePreviewValueAsDefault=true, inputName=NormalTexture_Offset_X, inputType=FunctionInput_Scalar, sortPriority=102, previewValue=<0 0 0 0>}
  n24 = FunctionInput() {bUsePreviewValueAsDefault=true, inputName=NormalTexture_Offset_Y, inputType=FunctionInput_Scalar, sortPriority=103, previewValue=<0 0 0 0>}
  n25 = FunctionInput() {bUsePreviewValueAsDefault=true, inputName=NormalTexture_Scale_X, inputType=FunctionInput_Scalar, sortPriority=104, previewValue=<1 0 0 0>}
  n26 = FunctionInput() {bUsePreviewValueAsDefault=true, inputName=NormalTexture_Scale_Y, inputType=FunctionInput_Scalar, sortPriority=105, previewValue=<1 0 0 0>}
  n27 = FunctionInput() {bUsePreviewValueAsDefault=true, description=Rad, inputName=NormalTexture_Rotation, inputType=FunctionInput_Scalar, sortPriority=106}
  n28 = MaterialFunctionCall(
   Offset_X=@n23
   Offset_Y=@n24
   Scale_X=@n25
   Scale_Y=@n26
   Rotation=@n27
   TexCoord=FunctionInput() {bUsePreviewValueAsDefault=true, inputName=NormalTexture_TexCoord, inputType=FunctionInput_Scalar, sortPriority=107}
   TilingMethod=FunctionInput() {bUsePreviewValueAsDefault=true, inputName=NormalTexture_TilingMethod, inputType=FunctionInput_Vector2, sortPriority=108}
  ) {materialFunction=&MF_TransformUVs}
  n29 = Constant3Vector() {constant=<rgba 0 0 1 0>}
  n30 = NamedRerouteUsage()
  FunctionOutput(None=StaticSwitchParameter(
    True=MaterialFunctionCall!(
     NormalTexture=@n20
     NormalScale=@n22
     NormalTexture_UVs=@n28.UVs
     NormalTexture_DDX=@n28.UVs
     NormalTexture_DDY=@n28.UVs
     NormalTexture_SinCos=@n28.UVs
    ) {materialFunction=/InterchangeAssets/gltf/MaterialFunctions/MF_Normals}.Normals
    False=MaterialFunctionCall!(Normal=@n29, Scale=@n30) {materialFunction=/InterchangeAssets/gltf/MaterialFunctions/MF_ScaleNormal}.Result
   ) {defaultValue=true, group=Configuration, parameterName=bHasNormalTexture, sortPriority=3}) {outputName=Normal, sortPriority=8}
  n31 = NamedRerouteUsage()
  FunctionOutput(None=StaticSwitch(
    True=@n19."Metallic (S)"
    False=MaterialFunctionCall!(In=@n31) {materialFunction=/InterchangeAssets/gltf/MaterialFunctions/MF_MaxComponentValue}.Result
    Value=StaticBoolParameter() {defaultValue=true, group=Configuration, parameterName=bHasDiffuseSpecGlossTexture, sortPriority=0}
   )) {outputName=Metallic, sortPriority=2}
  n32 = NamedRerouteUsage()
  FunctionOutput(None=StaticSwitch(
    True=@n19.Roughness
    False=OneMinus(None=@n32)
    Value=StaticBoolParameter() {defaultValue=true, group=Configuration, parameterName=bHasDiffuseSpecGlossTexture, sortPriority=0}
   )) {outputName=Roughness, sortPriority=3}
  n33 = StaticSwitch(
   True=@n18.Opacity
   False=NamedRerouteUsage()
   Value=StaticBoolParameter() {defaultValue=true, group=Configuration, parameterName=bHasDiffuseSpecGlossTexture, sortPriority=0}
  )
  FunctionOutput(None=Subtract(
    A=FunctionOutput(None=@n33) {outputName=Opacity, sortPriority=6}
    B=FunctionInput() {bUsePreviewValueAsDefault=true, inputName=AlphaCutoff, inputType=FunctionInput_Scalar, sortPriority=20}
   )) {outputName=OpacityMask, sortPriority=7}
  n34 = FunctionInput(Preview=TextureObject() {samplerType=SAMPLERTYPE_LinearColor, texture=/InterchangeAssets/gltf/Textures/T_White_Linear}) {bUsePreviewValueAsDefault=true, inputName=OcclusionTexture, inputType=FunctionInput_Texture2D, sortPriority=150}
  n35 = FunctionInput() {bUsePreviewValueAsDefault=true, inputName=OcclusionStrength, inputType=FunctionInput_Scalar, sortPriority=151, previewValue=<1 0 0 1>}
  n36 = NamedRerouteDeclaration(None=@n35) {name=Input_OcclusionStrength, nodeColor=<rgba 0.3359 0 1 1>}
  n37 = FunctionInput() {
   bUsePreviewValueAsDefault=true
   inputName=OcclusionTexture_Offset_X
   inputType=FunctionInput_Scalar
   sortPriority=152
   previewValue=<0 0 0 0>
  }
  n38 = FunctionInput() {
   bUsePreviewValueAsDefault=true
   inputName=OcclusionTexture_Offset_Y
   inputType=FunctionInput_Scalar
   sortPriority=153
   previewValue=<0 0 0 0>
  }
  n39 = FunctionInput() {bUsePreviewValueAsDefault=true, inputName=OcclusionTexture_Scale_X, inputType=FunctionInput_Scalar, sortPriority=154, previewValue=<1 0 0 0>}
  n40 = FunctionInput() {bUsePreviewValueAsDefault=true, inputName=OcclusionTexture_Scale_Y, inputType=FunctionInput_Scalar, sortPriority=155, previewValue=<1 0 0 0>}
  n41 = FunctionInput() {bUsePreviewValueAsDefault=true, description=Rad, inputName=OcclusionTexture_Rotation, inputType=FunctionInput_Scalar, sortPriority=156}
  n42 = FunctionInput() {bUsePreviewValueAsDefault=true, description=Rad, inputName=OcclusionTexture_TexCoord, inputType=FunctionInput_Scalar, sortPriority=157}
  n43 = FunctionInput() {bUsePreviewValueAsDefault=true, inputName=OcclusionTexture_TilingMethod, inputType=FunctionInput_Vector2, sortPriority=158}
  n44 = MaterialFunctionCall(Offset_X=@n37, Offset_Y=@n38, Scale_X=@n39, Scale_Y=@n40, Rotation=@n41, TexCoord=@n42, TilingMethod=@n43) {materialFunction=&MF_TransformUVs}
  FunctionOutput(None=StaticSwitchParameter(
    True=MaterialFunctionCall!(OcclusionTexture=@n34, OcclusionStrength=@n36, OcclusionTexture_UVs=@n44.UVs, OcclusionTexture_DDX=@n44.UVs, OcclusionTexture_DDY=@n44.UVs) {materialFunction=/InterchangeAssets/gltf/MaterialFunctions/MF_Occlusion}.Occlusion
    False=NamedRerouteUsage()
   ) {defaultValue=true, group=Configuration, parameterName=bHasOcclusionTexture, sortPriority=4}) {outputName=Occlusion, sortPriority=11}
  n45 = FunctionInput(Preview=TextureObject() {texture=&T_White_srgb}) {bUsePreviewValueAsDefault=true, inputName=EmissiveTexture, inputType=FunctionInput_Texture2D, sortPriority=51}
  n46 = NamedRerouteDeclaration(None=FunctionInput() {bUsePreviewValueAsDefault=true, inputName=EmissiveFactor, sortPriority=52}) {name=Input_EmissiveFactor, nodeColor=<rgba 0.6875 1 0 1>}
  n47 = FunctionInput() {bUsePreviewValueAsDefault=true, inputName=EmissiveStrength, inputType=FunctionInput_Scalar, sortPriority=53, previewValue=<1 0 0 1>}
  n48 = FunctionInput() {bUsePreviewValueAsDefault=true, inputName=EmissiveTexture_Offset_X, inputType=FunctionInput_Scalar, sortPriority=54, previewValue=<0 0 0 0>}
  n49 = FunctionInput() {bUsePreviewValueAsDefault=true, inputName=EmissiveTexture_Offset_Y, inputType=FunctionInput_Scalar, sortPriority=55, previewValue=<0 0 0 0>}
  n50 = FunctionInput() {bUsePreviewValueAsDefault=true, inputName=EmissiveTexture_Scale_X, inputType=FunctionInput_Scalar, sortPriority=56, previewValue=<1 0 0 0>}
  n51 = FunctionInput() {bUsePreviewValueAsDefault=true, inputName=EmissiveTexture_Scale_Y, inputType=FunctionInput_Scalar, sortPriority=57, previewValue=<1 0 0 0>}
  n52 = FunctionInput() {bUsePreviewValueAsDefault=true, description=Rad, inputName=EmissiveTexture_Rotation, inputType=FunctionInput_Scalar, sortPriority=58}
  n53 = MaterialFunctionCall(
   Offset_X=@n48
   Offset_Y=@n49
   Scale_X=@n50
   Scale_Y=@n51
   Rotation=@n52
   TexCoord=FunctionInput() {bUsePreviewValueAsDefault=true, inputName=EmissiveTexture_TexCoord, inputType=FunctionInput_Scalar, sortPriority=59}
   TilingMethod=FunctionInput() {bUsePreviewValueAsDefault=true, inputName=EmissiveTexture_TilingMethod, inputType=FunctionInput_Vector2, sortPriority=60}
  ) {materialFunction=&MF_TransformUVs}
  FunctionOutput(None=StaticSwitchParameter(
    True=MaterialFunctionCall!(
     EmissiveTexture=@n45
     EmissiveFactor=@n46
     EmissiveStrength=@n47
     EmissiveTexture_UVs=@n53.UVs
     EmissiveTexture_DDX=@n53.UVs
     EmissiveTexture_DDY=@n53.UVs
    ) {materialFunction=/InterchangeAssets/gltf/MaterialFunctions/MF_Emissive}.Emissive
    False=NamedRerouteUsage()
   ) {defaultValue=true, group=Configuration, parameterName=bHasEmissiveTexture, sortPriority=2}) {outputName=EmissiveColor, sortPriority=5}
 ]}
