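# Parent material for glTF assets using the specular-glossiness workflow in Interchange, Default Lit and usable on skeletal meshes and mesh deformers.
# Grouped parameters expose diffuse, specular-glossiness, emissive, normal and occlusion textures with factors, alpha cutoff and per-texture offset, scale, rotation, UV channel and tiling.
# All parameters feed one function, MF_SpecularGlossiness_Body, which holds the shading logic.
# Neutral default textures stand in for unset maps.
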Material /InterchangeAssets/gltf/M_SpecularGlossiness
graph {
 properties {
  bUsedWithMeshDeformer=true
  bUsedWithSkeletalMesh=true
  opacityMaskClipValue=0
  pixelDepthOffsetMode=PDOM_Legacy
  translucencyLightingMode=TLM_Surface
  usedShadingModels=Default Lit
 }
 nodes [
  n1 = TextureObjectParameter() {group=DiffuseSpecGloss, parameterName=DiffuseTexture, sortPriority=0, texture=/InterchangeAssets/gltf/Textures/T_White_srgb}
  n2 = TextureObjectParameter() {group=DiffuseSpecGloss, parameterName=SpecularGlossinessTexture, sortPriority=1, texture=&T_White_srgb}
  n3 = VectorParameter() {group=DiffuseSpecGloss, parameterName=DiffuseFactor, sortPriority=2, defaultValue=<rgba 1 1 1 1>}
  n4 = VectorParameter() {group=DiffuseSpecGloss, parameterName=DiffuseTexture_OffsetScale, sortPriority=5, defaultValue=<rgba 0 0 1 1>}
  n5 = VectorParameter() {group=DiffuseSpecGloss, parameterName=DiffuseTexture_TilingMethod, sortPriority=8, channelNames {a="", b="", g=Tiling V, r=Tiling U}}
  n6 = VectorParameter() {group=DiffuseSpecGloss, parameterName=SpecularGlossinessTexture_OffsetScale, sortPriority=9, defaultValue=<rgba 0 0 1 1>}
  n7 = VectorParameter() {
   group=DiffuseSpecGloss
   parameterName=SpecularGlossinessTexture_TilingMethod
   sortPriority=12
   channelNames {a="", b="", g=Tiling V, r=Tiling U}
  }
  n8 = VectorParameter() {group=Emissive, parameterName=EmissiveTexture_OffsetScale, sortPriority=3, defaultValue=<rgba 0 0 1 1>}
  n9 = VectorParameter() {group=Emissive, parameterName=EmissiveTexture_TilingMethod, sortPriority=6, channelNames {a="", b="", g=Tiling V, r=Tiling U}}
  n10 = TextureObjectParameter() {
   group=Normals
   parameterName=NormalTexture
   samplerType=SAMPLERTYPE_Normal
   sortPriority=0
   texture=/InterchangeAssets/gltf/Textures/T_Generic_N
  }
  n11 = VectorParameter() {group=Normals, parameterName=NormalTexture_OffsetScale, sortPriority=2, defaultValue=<rgba 0 0 1 1>}
  n12 = VectorParameter() {group=Normals, parameterName=NormalTexture_TilingMethod, sortPriority=5, channelNames {a="", b="", g=Tiling V, r=Tiling U}}
  n13 = TextureObjectParameter() {
   group=Occlusion
   parameterName=OcclusionTexture
   samplerType=SAMPLERTYPE_LinearColor
   sortPriority=0
   texture=/InterchangeAssets/gltf/Textures/T_White_Linear
  }
  n14 = VectorParameter() {group=Occlusion, parameterName=OcclusionTexture_OffsetScale, sortPriority=2, defaultValue=<rgba 0 0 1 1>}
  n15 = VectorParameter() {group=Occlusion, parameterName=OcclusionTexture_TilingMethod, sortPriority=5, channelNames {a="", b="", g=Tiling V, r=Tiling U}}
  MaterialFunctionCall!(
   DiffuseTexture=@n1
   SpecularGlossinessTexture=@n2
   DiffuseFactor_RGB=@n3.RGB
   DiffuseFactor_A=@n3.RGB
   SpecFactor=VectorParameter() {group=DiffuseSpecGloss, parameterName=SpecFactor, sortPriority=3, defaultValue=<rgba 1 1 1 0>}.RGB
   GlossinessFactor=ScalarParameter() {defaultValue=1, group=DiffuseSpecGloss, parameterName=GlossinessFactor, sliderMax=1, sortPriority=4}
   DiffuseTexture_Offset_X=@n4.R
   DiffuseTexture_Offset_Y=@n4.R
   DiffuseTexture_Scale_X=@n4.R
   DiffuseTexture_Scale_Y=@n4.R
   DiffuseTexture_Rotation=ScalarParameter() {group=DiffuseSpecGloss, parameterName=DiffuseTexture_Rotation, sliderMax=6.2832, sortPriority=6}
   DiffuseTexture_TexCoord=ScalarParameter() {group=DiffuseSpecGloss, parameterName=DiffuseTexture_TexCoord, sliderMax=3, sortPriority=7}
   DiffuseTexture_TilingMethod=@n5
   SpecularGlossinessTexture_Offset_X=@n6.R
   SpecularGlossinessTexture_Offset_Y=@n6.R
   SpecularGlossinessTexture_Scale_X=@n6.R
   SpecularGlossinessTexture_Scale_Y=@n6.R
   SpecularGlossinessTexture_Rotation=ScalarParameter() {group=DiffuseSpecGloss, parameterName=SpecularGlossinessTexture_Rotation, sliderMax=6.2832, sortPriority=10}
   SpecularGlossinessTexture_TexCoord=ScalarParameter() {group=DiffuseSpecGloss, parameterName=SpecularGlossinessTexture_TexCoord, sliderMax=3, sortPriority=11}
   SpecularGlossinessTexture_TilingMethod=@n7
   AlphaCutoff=ScalarParameter() {group=DiffuseSpecGloss, parameterName=AlphaCutoff, sliderMax=1, sortPriority=13}
   EmissiveTexture=TextureObjectParameter() {group=Emissive, parameterName=EmissiveTexture, sortPriority=0, texture=&T_White_srgb}
   EmissiveFactor=VectorParameter() {group=Emissive, parameterName=EmissiveFactor, sortPriority=1, defaultValue=<rgba 0 0 0 1>}.RGB
   EmissiveStrength=ScalarParameter() {defaultValue=1, group=Emissive, parameterName=EmissiveStrength, sortPriority=2}
   EmissiveTexture_Offset_X=@n8.R
   EmissiveTexture_Offset_Y=@n8.R
   EmissiveTexture_Scale_X=@n8.R
   EmissiveTexture_Scale_Y=@n8.R
   EmissiveTexture_Rotation=ScalarParameter() {group=Emissive, parameterName=EmissiveTexture_Rotation, sliderMax=6.2832, sortPriority=4}
   EmissiveTexture_TexCoord=ScalarParameter() {group=Emissive, parameterName=EmissiveTexture_TexCoord, sliderMax=3, sortPriority=5}
   EmissiveTexture_TilingMethod=@n9
   NormalTexture=@n10
   NormalScale=ScalarParameter() {defaultValue=1, group=Normals, parameterName=NormalScale, sortPriority=1}
   NormalTexture_Offset_X=@n11.R
   NormalTexture_Offset_Y=@n11.R
   NormalTexture_Scale_X=@n11.R
   NormalTexture_Scale_Y=@n11.R
   NormalTexture_Rotation=ScalarParameter() {group=Normals, parameterName=NormalTexture_Rotation, sliderMax=6.2832, sortPriority=3}
   NormalTexture_TexCoord=ScalarParameter() {group=Normals, parameterName=NormalTexture_TexCoord, sliderMax=3, sortPriority=4}
   NormalTexture_TilingMethod=@n12
   OcclusionTexture=@n13
   OcclusionStrength=ScalarParameter() {defaultValue=1, group=Occlusion, parameterName=OcclusionStrength, sliderMax=1, sortPriority=1}
   OcclusionTexture_Offset_X=@n14.R
   OcclusionTexture_Offset_Y=@n14.R
   OcclusionTexture_Scale_X=@n14.R
   OcclusionTexture_Scale_Y=@n14.R
   OcclusionTexture_Rotation=ScalarParameter() {group=Occlusion, parameterName=OcclusionTexture_Rotation, sliderMax=6.2832, sortPriority=3}
   OcclusionTexture_TexCoord=ScalarParameter() {group=Occlusion, parameterName=OcclusionTexture_TexCoord, sliderMax=3, sortPriority=4}
   OcclusionTexture_TilingMethod=@n15
  ) {materialFunction=/InterchangeAssets/gltf/MaterialBodies/MF_SpecularGlossiness_Body}
 ]
}
