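# Material function, exposed to the library, that converts glTF Specular-Glossiness diffuse and specular colors to Metallic-Roughness base color and metallic.
# Metallic comes from a quadratic solve on perceived brightness with 0.04 dielectric specular, and is zero below that threshold.
# Base color blends diffuse-derived and specular-derived estimates by squared metallic.
# Glossiness and roughness are not handled, and the specular-strength max reads only red and blue, skipping green.
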MaterialFunction /InterchangeAssets/gltf/SpecGlossToMetalRoughness
graph {
 properties {
  bExposeToLibrary=true
  description="Converts from the Specular-Glossiness PBR workflow to Metallic-Roughness, according to the GLTF reference implementation"
 }
 nodes [
  n1 = FunctionInput() {
   description="Linear-encoded diffuse color of material. Metals have a diffuse value of [0, 0, 0]. Already weighted by diffuseFactor."
   inputName=DiffuseColor
  }
  n2 = FunctionInput() {description=Linear-encoded specular color of the material. Already weighted by specularFactor, inputName=SpecularColor}
  n3 = MaterialFunctionCall(Float3=@n2) {materialFunction=&BreakOutFloat3Components}
  n4 = OneMinus(None=Max(A=Max(A=@n3.R, B=@n3.R), B=@n3.B)) {desc=oneMinusSpecularStrength}
  n5 = Multiply(A=MaterialFunctionCall!(RGBColor=@n1) {materialFunction=&/PerceivedBrightness}."Grayscale (S)", B=@n4)
  n6 = Constant() {desc=dielectricSpecular, r=0.04}
  n7 = OneMinus(None=@n6)
  n8 = MaterialFunctionCall(RGBColor=@n2) {materialFunction=&PerceivedBrightness}
  n9 = Subtract(A=Add(A=Divide(A=@n5, B=@n7), B=@n8."Grayscale (S)"), B=Multiply(A=@n6, B=Constant() {r=2})) {desc=b}
  n10 = Multiply(A=@n9, B=@n9)
  n11 = Multiply(A=@n6, B=Constant() {r=2}) {desc=a}
  n12 = Multiply(A=Multiply(A=Constant() {r=4}, B=@n11), B=Subtract(A=@n6, B=@n8."Grayscale (S)") {desc=c})
  n13 = SquareRoot(None=Max(A=Subtract(A=@n10, B=@n12), B=Constant()) {desc=D})
  n14 = Saturate(None=Divide(A=Subtract(A=@n13, B=@n9), B=@n11))
  n15 = If(A=@n8."Grayscale (S)", B=@n6, "A > B"=@n14, "A == B"=@n14, "A < B"=Constant()) {desc=metallic}
  FunctionOutput(None=@n15) {description=Converted metallic, outputName="Metallic (S)"}
  n16 = Multiply(A=@n15, B=@n15)
  n17 = OneMinus(None=@n6)
  n18 = OneMinus(None=@n15)
  n19 = Constant() {desc=epsilon, r=0}
  n20 = Multiply(A=@n1, B=Divide(A=Divide(A=@n4, B=@n17), B=Max(A=@n18, B=@n19))) {desc=baseColorFromDiffuse}
  n21 = OneMinus(None=@n15)
  n22 = Multiply(A=Divide(A=Constant() {r=1}, B=Max(A=@n15, B=@n19)), B=Subtract(A=@n2, B=Multiply(A=@n6, B=@n21))) {desc=diffuseFromSpecular}
  FunctionOutput(None=Saturate(None=LinearInterpolate(A=@n20, B=@n22, Alpha=@n16))) {description=Linear converted base color, outputName="BaseColor (V3)"}
 ]
}
